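# Material function for the Summer Popsicle backpack that adds parameter-driven frost and melting to incoming material attributes.
# SP_MeltedAmount lerps roughness and specular toward melted values and flattens the normal.
# Base color mixes the Diffuse texture with the incoming color by frost amount, then pulls back toward Diffuse by melted saturation and a separate frost-melt amount.
# The FX mask's red channel blends the result over the untouched input.
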MaterialFunction /Game/Accessories/FORT_Backpacks/Backpack_F_MED_Summer_Popsicle/Materials/MF_Summer_Popsicle
graph {nodes [
  n1 = GetMaterialAttributes(None=FunctionInput() {inputType=FunctionInput_MaterialAttributes}) {attributeGetTypes [
    #1
    #2
    #3
    #4
   ]}
  n2 = TextureSampleParameter2D() {
   parameterName=Diffuse
   texture=/Game/Accessories/FORT_Backpacks/Backpack_F_MED_Summer_Popsicle/Textures/T_F_MED_Summer_Popsicle_Backpack_Purple_D
  }
  n3 = LinearInterpolate(A=@n2.RGB, B=@n1.BaseColor, Alpha=ScalarParameter() {defaultValue=1, parameterName=SP_FrostAmount})
  n4 = ScalarParameter() {defaultValue=0.9, parameterName=SP_Melted_Saturation}
  n5 = SetMaterialAttributes!(
   MaterialAttributes=@n1.MaterialAttributes
   Roughness=LinearInterpolate(
    A=@n1.Roughness
    B=ScalarParameter() {defaultValue=0.1, parameterName=SP_Melted_Roughness}
    Alpha=ScalarParameter() {parameterName=SP_MeltedAmount}
   )
   Normal=MaterialFunctionCall(Normal=@n1.Normal, Flatness=ScalarParameter() {parameterName=SP_MeltedAmount}) {materialFunction=&FlattenNormal}.Result
   Specular=LinearInterpolate(
    A=@n1.Specular
    B=ScalarParameter() {defaultValue=1, parameterName=SP_Melted_Specular}
    Alpha=ScalarParameter() {parameterName=SP_MeltedAmount}
   )
   BaseColor=LinearInterpolate(A=@n3, B=LinearInterpolate(A=@n3, B=@n2.RGB, Alpha=@n4), Alpha=ScalarParameter() {parameterName=SP_MeltedAmount_Frost})
  ) {attributeSetTypes [
    #1
    #2
    #3
    #4
   ]}
  n6 = TextureSampleParameter2D() {
   parameterName=SP_FXMask
   samplerType=SAMPLERTYPE_Masks
   texture=/Game/Accessories/FORT_Backpacks/Backpack_F_MED_Summer_Popsicle/Textures/T_F_MED_Summer_Popsicle_Backpack_Purple_FX
  }
  FunctionOutput(None=SetMaterialAttributes!(MaterialAttributes=BlendMaterialAttributes(A=@n1.MaterialAttributes, B=@n5, Alpha=@n6.R)))
  Reroute(None=@n1.BaseColor)
  TextureSample() {
   samplerType=SAMPLERTYPE_Normal
   texture=/Game/Accessories/FORT_Backpacks/Backpack_F_MED_Summer_Popsicle/Textures/T_F_MED_Summer_Popsicle_Backpack_Purple_N
  }
 ]}
